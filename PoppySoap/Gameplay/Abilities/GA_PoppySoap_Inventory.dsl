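; Gameplay ability for PoppySoap that triggers a voice-over announcement when an item is added to the player's inventory.
; On activation it resolves the avatar's Fort player pawn and controller, then finds the controller's PoppySoap gameplay VO component.
; If the weapon gameplay tag and announcement gameplay effect class are valid, it passes both to that component's AnnounceItemAddtoInventory.
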(event Ability|EventActivateAbility
  (bind _weapongameplaytag (Variables|Default|GetWeaponGameplayTag))
  (bind _inventoryaddannouncementgameplayeffect (Variables|Default|GetInventoryAddAnnouncementGameplayEffect))
  (bind _outfortplayerpawn (CallFunction|GetAvatarasValidFortPlayerPawn)
    (:ValidFortPlayerPawn
      (bind _returnvalue (Fortnite|GetFortPlayerControllerfromActor _outfortplayerpawn))
      (Utilities|IsValid _returnvalue
        (:Is%20Valid
          (bind _returnvalue_2 (Actor|GetComponentbyClass _returnvalue "/PoppySoap/ConditionalGameplayVO/B_PoppySoapGameplayVOControllerComponent.B_PoppySoapGameplayVOControllerComponent_C"))
          (Utilities|IsValid _returnvalue_2
            (:Is%20Valid
              (if (and (GameplayTags|IsGameplayTagValid _weapongameplaytag) (Utilities|IsValidClass _inventoryaddannouncementgameplayeffect))
                (Class|BPoppySoapGameplayVOControllerComponent|AnnounceItemAddtoInventory _returnvalue_2 _weapongameplaytag _inventoryaddannouncementgameplayeffect)))))))))
