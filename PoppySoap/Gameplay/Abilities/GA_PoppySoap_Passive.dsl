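; PoppySoap passive ability that maintains non-authority (client) loose tags and re-triggers held attack inputs.
; While falling, a hotfix-interval timer probes floor distance on descent to toggle a too-close-for-air-kick tag.
; A crouch press while uncrouched adds a just-pressed tag, cleared by retriggerable delay.
; Melee tag-removal tasks replay held buttons through secondary fire or simulated Mantis primary input; ending clears timer, binding and tags.
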(event Ability|EventActivateAbility
  (bind _hotfix_airkickgrounddistancetracedetectinterval (Variables|Hotfix|GetHotfix_AirKickGroundDistanceTraceDetectInterval))
  (bind _returnvalue (ScalableFloat|GetValueatLevel _hotfix_airkickgrounddistancetracedetectinterval))
  ;[0.02034 0.15 0.03664 0.5] ActivateAbility: On the client, bind to know when the player presses crouch.
  (bind _outfortplayerpawn (CallFunction|GetAvatarasValidFortPlayerPawn)
    (:ValidFortPlayerPawn
      (Variables|Transient|SetAvatarPlayerPawn _outfortplayerpawn)
      (Ability|HasAuthority
        (:False
          (Input|BindEventtoOnCrouchInputJustPressed (Variables|Transient|GetAvatarPlayerPawn) (EventDispatchers|CreateEvent :!event Custom|HandleCrouchButtonPressed))
          (Utilities|FlowControl|Sequence
            (:then_0
              (if (> _returnvalue 0)
                (bind (_asynctaskproxy _tagcount) (Ability|Tasks|WaitGameplayTagCountChange "(TagName=\"MovementMode.Falling\")")
                  (:then
                    (Variables|Transient|SetAsyncTask_MovementModeFallingChanged _asynctaskproxy))
                  (:TagCountChanged
                    (if (> _tagcount 0)
                      (bind _returnvalue_2 (Utilities|Time|SetTimerbyEvent (EventDispatchers|CreateEvent :!event Custom|UpdateGroundDistanceTagForAirKick) _returnvalue true true))
                      (Variables|Transient|SetAirKickGroundDistanceCheckTimerHandle _returnvalue_2)
                      (CallFunction|UpdateGroundDistanceTagforAirKick)
                      (else
                        (CallFunction|StopCheckingGroundDistanceforAirKick)))))))
            (:then_1
              ;;[0.1242 0.03298 0.15 0.5] Simulate inputs if buttons are held.
              (bind _asynctaskproxy_2 (Ability|Tasks|WaitGameplayTagRemove "(TagName=\"Ability.Melee.Basic\")")
                (:then
                  (Variables|Transient|SetAsyncTask_WaitPrimaryEnd _asynctaskproxy_2)
                  (bind _asynctaskproxy_3 (Ability|Tasks|WaitGameplayTagRemove "(TagName=\"Ability.Melee.Secondary\")")
                    (:then
                      (Variables|Transient|SetAsyncTask_WaitSecondaryEnd _asynctaskproxy_3))
                    (:Removed
                      (Utilities|FlowControl|DelayUntilNextTick))))
                (:Removed
                  (bind _outfortplayerpawn_2 (CallFunction|GetAvatarasValidFortPlayerPawn)
                    (:ValidFortPlayerPawn
                      (if (Variables|/Script/FortniteGame.FortPawn|GetbSecondaryInputHeld _outfortplayerpawn_2)
                        (Weapon|PawnStartSecondaryFire _outfortplayerpawn_2 0))))))))
          (bind _outfortplayerpawn_3 (CallFunction|GetAvatarasValidFortPlayerPawn)
            (:ValidFortPlayerPawn
              (if (Variables|/Script/FortniteGame.FortPawn|GetbPrimaryInputHeld _outfortplayerpawn_3)
                (bind _returnvalue_3 (Actor|GetComponentbyClass _outfortplayerpawn_3 "/Script/MantisRuntime.FortMantisPawnComponent"))
                (Utilities|IsValid _returnvalue_3
                  (:Is%20Valid
                    (Mantis|SimulateInputEvent _returnvalue_3 "Primary" "Pressed" true false)))))))))))

(event Custom|UpdateGroundDistanceTagForAirKick
  (bind _tooclosetogroundforairkickgameplaytag (Variables|Settings|GetTooCloseToGroundForAirKickGameplayTag))
  (bind _returnvalue (GameplayTags|MakeGameplayTagContainerfromTag _tooclosetogroundforairkickgameplaytag))
  ;[0.15 0.03169 0.1316 0.5] UpdateGroundDistanceTagForAirKick: Apply or remove gameplay tag on the player that indicates if they are too close to the ground to execute an air kick.
  (Ability|HasAuthority
    (:False
      (bind _avatarplayerpawn (Variables|Transient|GetAvatarPlayerPawn)
        (:then
          (bind _charactermovement (Variables|/Script/Engine.Character|GetCharacterMovement _avatarplayerpawn)
            (:then
              (bind _capsulecomponent (Variables|/Script/Engine.Character|GetCapsuleComponent _avatarplayerpawn)
                (:then
                  (if (< (.z (Transformation|GetVelocity _avatarplayerpawn)) 0)
                    (if (Utilities|Struct|BreakFindFloorResult (Pawn|Components|CharacterMovement|ComputeFloorDistance _charactermovement (Transformation|GetWorldLocation _capsulecomponent) 0.0 (Variables|Settings|GetTooCloseToGroundForAirKickDistance) (Components|Capsule|GetScaledCapsuleRadius _capsulecomponent)))
                      (if (not (GameplayTags|HasMatchingGameplayTag _avatarplayerpawn _tooclosetogroundforairkickgameplaytag))
                        (Ability|GameplayEffect|AddLooseGameplayTags _avatarplayerpawn _returnvalue))
                      (else
                        _))
                    (else
                      _))
                  (Ability|GameplayEffect|RemoveLooseGameplayTags _avatarplayerpawn _returnvalue))
                (:else)))
            (:else)))
        (:else)))))

(event Custom|StopCheckingGroundDistanceForAirKick
  ;[0.15 0.05636 0.1419 0.5] StopCheckingGroundDistanceForAirKick: Stop timer that runs trace every interval to see how close the player is to the ground.
  ; Remove the gameplay tag on the player that indicates that the player is too close to the ground to execute air kick.
  (Ability|HasAuthority
    (:False
      (Utilities|Time|ClearandInvalidateTimerbyHandle (Variables|Transient|GetAirKickGroundDistanceCheckTimerHandle))
      (Ability|GameplayEffect|RemoveLooseGameplayTags (Variables|Transient|GetAvatarPlayerPawn) (GameplayTags|MakeGameplayTagContainerfromTag (Variables|Settings|GetTooCloseToGroundForAirKickGameplayTag))))))

(event Custom|HandleCrouchButtonPressed
  (bind _crouchjustpressedgameplaytag (Variables|Settings|GetCrouchJustPressedGameplayTag))
  (bind _returnvalue (GameplayTags|MakeGameplayTagContainerfromTag _crouchjustpressedgameplaytag))
  (bind _avatarplayerpawn (Variables|Transient|GetAvatarPlayerPawn))
  ;[0.15 0.0263 0.106 0.5] HandleCrouchButtonPressed: Add gameplay tag to the player that marks that they have just pressed the crouch button.
  (Ability|HasAuthority
    (:False
      (bind _avatarplayerpawn_2 (Variables|Transient|GetAvatarPlayerPawn)
        (:then
          (if (not (Variables|/Script/Engine.Character|GetbIsCrouched _avatarplayerpawn_2))
            (if (not (GameplayTags|HasMatchingGameplayTag _avatarplayerpawn _crouchjustpressedgameplaytag))
              (Ability|GameplayEffect|AddLooseGameplayTags _avatarplayerpawn _returnvalue)
              (Utilities|FlowControl|RetriggerableDelay (Variables|Settings|GetCrouchJustPressedCacheTime))
              (Ability|GameplayEffect|RemoveLooseGameplayTags (Variables|Transient|GetAvatarPlayerPawn) _returnvalue))))
        (:else)))))

(event Ability|EventOnEndAbility
  (params (bWasCancelled :type Boolean))
  ;[0.15 0.07497 0.01623 0.5] EndAbility: Clean up.
  (Utilities|FlowControl|Sequence
    (:then_0
      (Ability|HasAuthority
        (:False
          (CallFunction|StopCheckingGroundDistanceforAirKick)
          (GameplayTasks|EndTask (Variables|Transient|GetAsyncTask_MovementModeFallingChanged))
          (bind _avatarplayerpawn (Variables|Transient|GetAvatarPlayerPawn)
            (:then
              (Input|UnbindEventfromOnCrouchInputJustPressed _avatarplayerpawn (EventDispatchers|CreateEvent :!event Custom|HandleCrouchButtonPressed))
              (Ability|GameplayEffect|RemoveLooseGameplayTags _avatarplayerpawn (GameplayTags|MakeGameplayTagContainerfromTag (Variables|Settings|GetCrouchJustPressedGameplayTag))))
            (:else)))))
    (:then_1
      (|CallParentFunction bWasCancelled))))
